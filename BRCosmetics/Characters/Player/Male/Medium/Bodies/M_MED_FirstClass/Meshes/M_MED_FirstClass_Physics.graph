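# Physics asset for the M_MED_FirstClass male medium player cosmetic body, set up for secondary motion on the shirt's front, back, left and right bones.
# Each dyn_shirt bone has a larger offset capsule constrained to its base_shirt bone with limited swing and zero twist range.
# The four base_shirt bones have tiny capsules and join to spine_02 with free swing and twist, and no other body bones are included.
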PhysicsAsset /BRCosmetics/Characters/Player/Male/Medium/Bodies/M_MED_FirstClass/Meshes/M_MED_FirstClass_Physics
bodies {
 base_shirt_b [{shapeType=Capsule, length=0.5, radius=0.5, shapeName=base_shirt_b_capsule}]
 base_shirt_f [{shapeType=Capsule, length=0.5, radius=0.5, shapeName=base_shirt_f_capsule}]
 base_shirt_l [{shapeType=Capsule, length=0.5, radius=0.5, shapeName=base_shirt_l_capsule}]
 base_shirt_r [{shapeType=Capsule, length=0.5, radius=0.5, shapeName=base_shirt_r_capsule}]
 dyn_shirt_b [{shapeType=Capsule, length=4, radius=4, shapeName=dyn_shirt_b_capsule, center=<2 0 0>}]
 dyn_shirt_f [{shapeType=Capsule, length=4, radius=4, shapeName=dyn_shirt_f_capsule, center=<2 0 0>}]
 dyn_shirt_l [{shapeType=Capsule, length=4, radius=4, shapeName=dyn_shirt_l_capsule, center=<2 0 0>}]
 dyn_shirt_r [{shapeType=Capsule, length=4, radius=4, shapeName=dyn_shirt_r_capsule, center=<2 0 0>}]
 spine_02 [{shapeType=Capsule, length=0.5, radius=0.5, shapeName=spine_02_capsule}]
}
constraints [
 {bone1Name=dyn_shirt_f, bone2Name=base_shirt_f, swing1LimitDegrees=45, swing2LimitDegrees=45, twistLimitDegrees=0}
 {bone1Name=dyn_shirt_b, bone2Name=base_shirt_b, swing1LimitDegrees=45, swing2LimitDegrees=45, twistLimitDegrees=0}
 {bone1Name=dyn_shirt_l, bone2Name=base_shirt_l, swing1LimitDegrees=45, swing2LimitDegrees=45, twistLimitDegrees=0}
 {bone1Name=dyn_shirt_r, bone2Name=base_shirt_r, swing1LimitDegrees=45, swing2LimitDegrees=45, twistLimitDegrees=0}
 {
  bone1Name=base_shirt_f
  bone2Name=spine_02
  swing1LimitDegrees=45
  swing1Motion=Free
  swing2LimitDegrees=45
  swing2Motion=Free
  twistLimitDegrees=45
  twistMotion=Free
 }
 {
  bone1Name=base_shirt_b
  bone2Name=spine_02
  swing1LimitDegrees=45
  swing1Motion=Free
  swing2LimitDegrees=45
  swing2Motion=Free
  twistLimitDegrees=45
  twistMotion=Free
 }
 {
  bone1Name=base_shirt_l
  bone2Name=spine_02
  swing1LimitDegrees=45
  swing1Motion=Free
  swing2LimitDegrees=45
  swing2Motion=Free
  twistLimitDegrees=45
  twistMotion=Free
 }
 {
  bone1Name=base_shirt_r
  bone2Name=spine_02
  swing1LimitDegrees=45
  swing1Motion=Free
  swing2LimitDegrees=45
  swing2Motion=Free
  twistLimitDegrees=45
  twistMotion=Free
 }
]
