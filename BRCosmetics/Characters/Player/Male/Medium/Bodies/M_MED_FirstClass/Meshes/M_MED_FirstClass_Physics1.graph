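# Physics asset for the necklace on the M_MED_FirstClass male player body, covering only the necklace bones and two collider bones.
# A capsule chain runs from base_necklace through dyn_necklace_01 and 02, then forks into 03 to 06 and 04 to 05, with twin parallel capsules on 06.
# Constraints set narrow swing limits and zero-degree twist limits.
# Two large, unconstrained necklace_Collider capsules appear to act as body collision for the chain.
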PhysicsAsset /BRCosmetics/Characters/Player/Male/Medium/Bodies/M_MED_FirstClass/Meshes/M_MED_FirstClass_Physics1
bodies {
 base_necklace [{shapeType=Capsule, length=1.6821, radius=0.5, shapeName=base_necklace_capsule, center=<1.6149 0 0>, rotation=<rot 90 0 0>}]
 dyn_necklace_01 [{shapeType=Capsule, length=0.5, radius=0.5, shapeName=dyn_necklace_01_capsule, center=<1 0 0>, rotation=<rot 90 0 0>}]
 dyn_necklace_02 [{shapeType=Capsule, length=1.062, radius=0.5, shapeName=dyn_necklace_02_capsule, center=<1.1931 0 0>, rotation=<rot 90 0 0>}]
 dyn_necklace_03 [{shapeType=Capsule, length=0.5, radius=0.5, shapeName=dyn_necklace_03_capsule, center=<1.4253 0 0>, rotation=<rot 90 0 0>}]
 dyn_necklace_04 [{shapeType=Capsule, length=0.505, radius=0.505, shapeName=dyn_necklace_04_capsule, center=<1 0 0>, rotation=<rot 90 0 0>}]
 dyn_necklace_05 [{shapeType=Capsule, length=0.8014, radius=1.8413, shapeName=dyn_necklace_05_capsule, center=<2.8453 0 0.9058>, rotation=<rot 82.2637 0 0>}]
 dyn_necklace_06 [
  {shapeType=Capsule, length=4.7523, radius=0.6177, shapeName=dyn_necklace_06_capsule, center=<3.7 0.78 0.129>, rotation=<rot 88 0 0>}
  {shapeType=Capsule, length=4.7523, radius=0.6177, shapeName=dyn_necklace_06_capsule, center=<3.7 -0.78 0>, rotation=<rot 88 0 0>}
 ]
 necklace_Collider1 [{
   shapeType=Capsule
   length=10.9838
   radius=11.3024
   shapeName=necklace_Collider1_capsule
   center=<-2.0188 1.4124 -10.8138>
   rotation=<rot 0.0002 -0.0001 68.5903>
  }]
 necklace_Collider2 [{
   shapeType=Capsule
   length=8.648
   radius=11.8089
   shapeName=necklace_Collider2_capsule
   center=<0.2268 0.9915 -11.2342>
   rotation=<rot 0.0003 0 85.8196>
  }]
}
constraints [
 {bone1Name=dyn_necklace_01, bone2Name=base_necklace, swing1LimitDegrees=15, swing2LimitDegrees=15, twistLimitDegrees=0}
 {bone1Name=dyn_necklace_02, bone2Name=dyn_necklace_01, swing1LimitDegrees=10, swing2LimitDegrees=8, twistLimitDegrees=0}
 {bone1Name=dyn_necklace_03, bone2Name=dyn_necklace_02, swing1LimitDegrees=15, swing2LimitDegrees=25, twistLimitDegrees=0}
 {bone1Name=dyn_necklace_05, bone2Name=dyn_necklace_04, swing1LimitDegrees=15, swing2LimitDegrees=25, twistLimitDegrees=0}
 {bone1Name=dyn_necklace_06, bone2Name=dyn_necklace_03, swing1LimitDegrees=5, swing2LimitDegrees=25, twistLimitDegrees=0}
 {bone1Name=dyn_necklace_04, bone2Name=dyn_necklace_02, swing1LimitDegrees=10, swing2LimitDegrees=25, twistLimitDegrees=0}
]
