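# Post-process style Anim Blueprint for the FirstClass male body cosmetic, layering corrective rigs and secondary motion onto the parent mesh's pose.
# Copied pose runs through two Control Rigs, necklace collider constraints, two RigidBody simulations and bracelet AnimDynamics.
# Event graph turns pawn movement, targeting and weapon state into rig alphas and computes gravity override.
# It also drives WPO cloth material parameters from feet, calves and skydiving, and flags FirstClass backpack variants.
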AnimBlueprint /BRCosmetics/Characters/Player/Male/Medium/Bodies/M_MED_FirstClass/Meshes/M_MED_FirstClass_AnimBP
target_skeleton=&/M_MED_FirstClass_Skeleton
graphs [
 AnimGraph {nodes [
   "Transform (Modify) Bone - Bone: dyn_necklace_02"(Alpha=Get IsTargeting() -> (IsTargeting:double), Translation="0,0,0.8", Rotation="0,0,0", Scale="1,1,1") {
    rotationMode=BMM_Additive
    rotationSpace=BCS_BoneSpace
    translationMode=BMM_Additive
    translationSpace=BCS_BoneSpace
    alphaBoolBlend {blendOption=Linear}
    alphaScaleBiasClamp {bInterpResult=true, interpSpeedDecreasing=5, interpSpeedIncreasing=5}
    boneToModify {boneName=dyn_necklace_02}
    translation {z=0.8}
   }
   n1 = Control Rig(Source=Copy Pose From Mesh() {bUseAttachedParent=true}) {
    alphaBoolBlend {blendOption=Linear}
    alphaScaleBiasClamp {bInterpResult=true}
    controlRigAssetReference {blueprintRigClass=&/M_MED_DeformRig_FirstClass_ControlRig.M_MED_DeformRig_FirstClass_ControlRig_C}
   }
   n2 = Control Rig!(
    Source=@n1
    Alpha
    Skydive_Alpha:double=Get isSkydiving() -> (isSkydiving:double)
    Crouch_Alpha_0:double=Get crouching() -> (crouching:double)
    Target_Alpha_0:double=Get HG() -> (HG:double)
    ShirtFix_Alpha:double=Get ShirtFix() -> (ShirtFix:double)
   ) -> (Pose) {
    alphaBoolBlend {blendOption=Linear}
    alphaScaleBiasClamp {bInterpResult=true}
    controlRigAssetReference {blueprintRigClass=&/M_MED_FirstClass_CtrlRig.M_MED_FirstClass_CtrlRig_C}
   }
   n3 = "Constraint - necklace_Collider1 (PPPP)"!(
    ComponentPose=Local To Component(LocalPose=@n2)
    Alpha=1
    ConstraintWeights_0=0.4836
    ConstraintWeights_1:float=0.3963
    ConstraintWeights_2:float=0.1124
    ConstraintWeights_3:float=0.0077
   ) -> (Pose) {
    alphaBoolBlend {blendOption=Linear}
    boneToModify {boneName=necklace_Collider1}
    constraintSetup [{targetBone {boneName=spine_05}}, {targetBone {boneName=spine_04}}, {targetBone {boneName=spine_03}}, {targetBone {boneName=clavicle_l}}]
    constraintWeights [1,1,1,1]
   }
   n4 = "Constraint - necklace_Collider2 (PPPP)"!(
    ComponentPose=@n3
    Alpha=1
    ConstraintWeights_0=0.4908
    ConstraintWeights_1:float=0.2443
    ConstraintWeights_2:float=0.1894
    ConstraintWeights_3:float=0.0755
   ) -> (Pose) {
    alphaBoolBlend {blendOption=Linear}
    boneToModify {boneName=necklace_Collider2}
    constraintSetup [{targetBone {boneName=spine_04}}, {targetBone {boneName=spine_05}}, {targetBone {boneName=spine_03}}, {targetBone {boneName=spine_02}}]
    constraintWeights [1,1,1,1]
   }
   n5 = Get GravityOverride()
   n6 = RigidBody!(ComponentPose=@n4, Alpha, OverrideWorldGravity:Vector=@n5, ExternalForce) -> (Pose) {
    bClampLinearTranslationLimitToRefPose=true
    bOverrideWorldGravity=true
    overridePhysicsAsset=&/M_MED_FirstClass_Physics
    simulationSpace=BaseBoneSpace
    alphaBoolBlend {blendOption=Linear}
    baseBoneRef {boneName=spine_05}
    componentAppliedLinearAccClamp=<2000 2000 1000>
    componentLinearAccScale {x=0.4}
    componentLinearVelScale {x=0.4}
    overrideWorldGravity {z=-980}
    simSpaceSettings {maxAngularAcceleration=100, maxAngularVelocity=5, maxLinearAcceleration=300, maxLinearVelocity=50, velocityScaleZ=0, worldAlpha=1}
   }
   n7 = RigidBody(ComponentPose=@n6, OverrideWorldGravity=@n5) {
    bClampLinearTranslationLimitToRefPose=true
    bOverrideWorldGravity=true
    overridePhysicsAsset=&/M_MED_FirstClass_Physics1
    simulationSpace=BaseBoneSpace
    alphaBoolBlend {blendOption=Linear}
    baseBoneRef {boneName=spine_04}
    componentAppliedLinearAccClamp=<3000 3000 3000>
    componentLinearAccScale=<0.2 0.2 0.2>
    componentLinearVelScale=<0.2 0.2 0.2>
    overrideWorldGravity {z=-980}
    simSpaceSettings {
     maxAngularAcceleration=2
     maxAngularVelocity=1.2
     maxLinearAcceleration=2
     maxLinearVelocity=1
     velocityScaleZ=0
     worldAlpha=0.05
     externalLinearDragV=<0.2 0.2>
    }
   }
   n8 = Anim Dynamics - Bone: dyn_braclet(ComponentPose=@n7, Alpha=1, ExternalForce="0,0,0") {
    angularDampingOverride=1
    angularSpringConstant=20000
    bAngularSpring=true
    bOverrideAngularDamping=true
    bUseGravityOverride=true
    gravityScale=0.35
    simulationSpace=RootRelative
    alphaBoolBlend {blendOption=Linear}
    boundBone {boneName=dyn_braclet}
    componentLinearAccScale=<1 1 1>
    componentLinearVelScale=<1 1 1>
    gravityOverride {z=-980}
    physicsBodyDefinitions [{
      boundBone {boneName=dyn_braclet}
      boxExtents=<20 30 30>
      constraintSetup {angularLimitsMax=<10 10 10>, angularLimitsMin=<-10 -10 -10>, angularTarget=<1 0 0>}
      localJointOffset=<4 0 0>
     }]
   }
   Output Pose(Result=Component To Local(ComponentPose=@n8)) {layerGroup=None}
  ]}
 EventGraph {nodes [
   n1 = Event BlueprintUpdateAnimation()
   > n2 = Sequence!(execute) -> (then_0, then_1, then_2:Exec, then_3:Exec, then_4:Exec, then_5:Exec)
   > Set IsTargeting(IsTargeting:double="To Float (Boolean)"(InBool=Get bIsTargeting())) -> (Output_Get:double)
   n3 = NOT Boolean(A=Get bIsParachuteOpened())
   n4 = NOT Boolean(A=Get bIsSkydiveFloating())
   n5 = NOT Boolean(A=Get bIsSprinting())
   n6 = NOT Boolean(A=Get bIsSurfaceSwimming())
   > Set ShirtFix(ShirtFix:double="To Float (Boolean)"(InBool=AND Boolean!(A=@n4, B=@n5, C:bool=@n6, D:bool=@n3) -> (ReturnValue))) -> (Output_Get:double)
   > Set crouching(crouching:double="To Float (Boolean)"(InBool=Get bIsCrouching())) -> (Output_Get:double)
   n7 = Gravity Override State Function DEPRECATED(
    execute=@n2.then_1
    CharacterPartAnimInstance=Self-Reference()
    GameTimeInSeconds=GetGameTimeInSeconds()
    GravityOverrideParams=Get Gravity Override Params()
   )
   > Set GravityOverride(GravityOverride=@n7)
   Is Valid(exec=@n2.then_2, InputObject=Get PlayerPawn() -> (PlayerPawn:Fort Player Pawn Athena Object Reference))
   > Set WPO Globals Function(
    Material=Get Mat Instance Dynamic()
    Overlay Material=Get Overlay Mat Instance Dynamic()
    Skeletal Mesh=Get SkelMesh()
    "Local Noise Transform  Location"=Get Local Noise Transform Location() -> (Local Noise Transform Location:Vector)
    Local Noise Offset=Get Local Noise Offset()
   )
   > Set WPO Regional Function!(
    execute
    Material:Material Instance Dynamic Object Reference=Get Mat Instance Dynamic()
    Overlay Material:Material Instance Dynamic Object Reference=Get Overlay Mat Instance Dynamic()
    Skeletal Mesh:Skeletal Mesh Component Object Reference=Get SkelMesh()
    Velocity Transform:Name=foot_r
    Velocity Interp Speed:double=Get Velocity Interp Speed()
    Collision Transform:Name=calf_r
    Collision Transform Axis:EAxis Enum
    Invert Collision Transform Axis:bool=false
    Delta Time X:double=@n1.DeltaTimeX
    Previous Velocity Transform Location:"Vector (by ref)"=Get R1 Velocity Transform Location()
    Previous Velocity Transform Velocity:"Vector (by ref)"=Get R1 Velocity Transform Velocity()
    Region:int=1
   ) -> (then)
   > Set WPO Regional Function(
    Material=Get Mat Instance Dynamic()
    Overlay Material=Get Overlay Mat Instance Dynamic()
    Skeletal Mesh=Get SkelMesh()
    Velocity Transform=foot_l
    Velocity Interp Speed=Get Velocity Interp Speed()
    Collision Transform=calf_l
    Invert Collision Transform Axis=true
    Delta Time X=@n1.DeltaTimeX
    Previous Velocity Transform Location=Get R2 Velocity Transform Location()
    Previous Velocity Transform Velocity=Get R2 Velocity Transform Velocity()
    Region=2
   )
   > n8 = Branch(Condition=Get bIsSkyDiving())
   > Set WPO Pho Wind Intensity(WPO Pho Wind Intensity=0.44)
   > n9 = Set WPO Pho Turbulence(WPO Pho Turbulence:double=0.8) -> (Output_Get:double)
   n10 = Get Mat Instance Dynamic()
   n11 = Get Overlay Mat Instance Dynamic()
   SetScalarParameterValue(execute=[@n9.then, @n13.then], self=[@n10, @n11], ParameterName=ClothSim_WindStrength, Value=Get WPO Pho Wind Intensity())
   n12 = Get WPO Pho Turbulence() -> (WPO Pho Turbulence:double)
   > SetScalarParameterValue(self=[@n11, @n10], ParameterName=ClothSim_R1_TurbulenceStrength, Value=@n12)
   > SetScalarParameterValue(self=[@n11, @n10], ParameterName=ClothSim_R2_TurbulenceStrength, Value=@n12)
   Set WPO Pho Wind Intensity(execute=@n8.else, WPO Pho Wind Intensity=0)
   > n13 = Set WPO Pho Turbulence(WPO Pho Turbulence=0.2)
   n14 = Cast To FortPlayerPawn(execute=@n2.then_3, Object=GetOwningActor())
   > n15 = GetVariantChannelActiveIndexForItemActor!(
    execute
    Item:Athena Cosmetic Item Definition Object Reference=/BRCosmetics/Athena/Items/Cosmetics/Backpacks/Backpack_FirstClass
    Actor:Actor Object Reference=@n14
    Channel:Gameplay Tag Structure=(TagName="Cosmetics.Variant.Channel.Progressive")
    VariantList:Array of Gameplay Tag Structures=Make Array!(
     "[0]"=(TagName="")
     "[1]":Gameplay Tag Structure=(TagName="")
     "[2]":Gameplay Tag Structure=(TagName="")
     "[3]":Gameplay Tag Structure=(TagName="")
     "[4]":Gameplay Tag Structure=(TagName="")
     "[5]":Gameplay Tag Structure=(TagName="")
    ) -> (Array)
    bErrorIfChannelMissing:bool=false
    bErrorIfTagNotInArray:bool=false
   ) -> (then, Index:int=0, bIndexIsValid:bool, bFoundChannel:bool)
   > n16 = Sequence!(execute) -> (then_0, then_1, then_2, then_3, then_4)
   > Set James loaded(James loaded:bool=AND Boolean!(A="Equal (Integer)"(A=@n15), B=@n15.bIndexIsValid) -> (ReturnValue)) -> (Output_Get:bool)
   Set Kieran loaded(execute=@n16.then_1, Kieran loaded:bool=AND Boolean(A="Equal (Integer)"(A=@n15, B=1), B=@n15.bIndexIsValid)) -> (Output_Get:bool)
   Set Nadar loaded(execute=@n16.then_2, Nadar loaded:bool=AND Boolean(A="Equal (Integer)"(A=@n15, B=2), B=@n15.bIndexIsValid)) -> (Output_Get:bool)
   Set Yuta loaded(execute=@n16.then_3, Yuta loaded:bool=AND Boolean(A="Equal (Integer)"(A=@n15, B=3), B=@n15.bIndexIsValid)) -> (Output_Get:bool)
   Set Carlos Loaded(execute=@n16.then_4, Carlos Loaded:bool=AND Boolean(A="Equal (Integer)"(A=@n15, B=4), B=@n15.bIndexIsValid)) -> (Output_Get:bool)
   n17 = Cast To FortPlayerAnimInstance(execute=@n2.then_4, Object=GetAnimInstance(self=GetOwnerSkeletalMeshComponent()))
   > Set ar_firingPosition_HG(ar_firingPosition_HG:int=0) -> (Output_Get:int=0)
   > n18 = Switch on EFortWeaponCoreAnimation!(execute, Selection:EFortWeaponCoreAnimation Enum=Get WeaponCoreAnim(self=@n17)) -> (
    Melee:Exec
    Pistol:Exec
    Shotgun:Exec
    PaperBlueprint:Exec
    Rifle:Exec
    MeleeOneHand:Exec
    MachinePistol:Exec
    RocketLauncher:Exec
    GrenadeLauncher:Exec
    GoingCommando:Exec
    AssaultRifle:Exec
    TacticalShotgun:Exec
    SniperRifle:Exec
    TrapPlacement:Exec
    ShoulderLauncher:Exec
    AbilityDecoTool:Exec
    Crossbow:Exec
    C4:Exec
    RemoteControl:Exec
    DualWield:Exec
    AR_BullPup:Exec
    AR_ForwardGrip:Exec
    MedPackPaddles:Exec
    SMG_P90:Exec
    AR_DrumGun:Exec
    Consumable_Small:Exec
    Consumable_Large:Exec
    Balloon:Exec
    MountedTurret:Exec
    CreativeTool:Exec
    ExplosiveBow:Exec
    AshtonIndigo:Exec
    AshtonChicago:Exec
    MeleeDualWield:Exec
    Unarmed:Exec
   )
   n19 = Branch(execute=@n18.DualWield)
   > n20 = Set ar_firingPosition_HG(ar_firingPosition_HG=1) -> (Output_Get=0)
   > n21 = Set gotHG(gotHG:bool="To Boolean (Integer)"!(InInt:int=@n20) -> (ReturnValue:bool)) -> (Output_Get:bool)
   n22 = Get bIsTargeting()
   > Set HG(HG:double="To Float (Boolean)"(InBool=AND Boolean(A=@n21, B=@n22))) -> (Output_Get:double)
   > Set isSkydiving(isSkydiving:double="To Float (Boolean)"(InBool=Get bIsSkyDiving())) -> (Output_Get:double)
   Set ar_firingPosition_HG(execute=@n19.else, ar_firingPosition_HG=0) -> (Output_Get=0)
   Event OnCharacterCustomizationLoaded()
   > n23 = Set SkelMesh(SkelMesh=GetPartSkeletalMeshComponent())
   > n24 = Get Dynamic Material Instance(SkeletalMeshComponent=@n23)
   > Set Mat Instance Dynamic(Mat Instance Dynamic=@n24)
   > n25 = GetOrCreateDynamicOverlayMaterial(MeshComponent=@n23)
   > Set Overlay Mat Instance Dynamic(Overlay Mat Instance Dynamic=@n25)
   Event BlueprintInitializeAnimation()
   > Set SkelMesh(SkelMesh=GetOwnerSkeletalMeshComponent())
   n26 = TryGetPawnOwner()
   > Is Valid(InputObject=@n26)
   > n27 = Cast To FortPlayerPawnAthena(Object=@n26)
   > n28 = Set PlayerPawn(PlayerPawn:Fort Player Pawn Athena Object Reference=@n27) -> (Output_Get:Fort Player Pawn Athena Object Reference)
   > Set Body_Mesh(Body_Mesh=GetSkeletalMeshForPartType(self=@n28))
   SetScalarParameterValue(self=[@n11, @n10], ParameterName=ClothSim_R3_TurbulenceStrength, Value=@n12)
  ]}
]
variables [
 Alpha_IsTargeting_p {type=number}
 ar_firingPosition_HG {type=integer}
 Body_Mesh {type=SkeletalMeshComponent}
 Carlos Loaded {type=boolean}
 crouching {type=number}
 crouchingWalking {type=number}
 gotHG {type=boolean}
 Gravity Override Params {type=GravityOverrideParamsStruct}
 GravityOverride {type=Vector}
 HG {type=number}
 isSkydiving {type=number}
 IsTargeting {type=number}
 James loaded {type=boolean}
 Kieran loaded {type=boolean}
 Local Noise Offset {type=Vector}
 Local Noise Transform Location {type=Vector}
 Mat Instance Dynamic {type=MaterialInstanceDynamic}
 Nadar loaded {type=boolean}
 Overlay Mat Instance Dynamic {type=MaterialInstanceDynamic}
 PlayerPawn {type=FortPlayerPawnAthena}
 R1 Velocity Transform Location {type=Vector}
 R1 Velocity Transform Velocity {type=Vector}
 R2 Velocity Transform Location {type=Vector}
 R2 Velocity Transform Velocity {type=Vector}
 ShirtFix {type=number}
 SkelMesh {type=SkeletalMeshComponent}
 Velocity Interp Speed {type=number}
 WPO Pho Turbulence {type=number}
 WPO Pho Wind Intensity {type=number}
 Yuta loaded {type=boolean}
]
